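# Sound cue for the cooking sound of the Port-A-Shack gadget item.
# A single wave player for PortaShack_Cook feeds a modulator at the root.
# It plays in the gadgets sound class at reduced volume and borrows the default grenade cook attenuation for distance falloff.
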SoundCue /PortAShackGameplay/Items/Audio/PortaShack_Cook_Cue
root=Modulator_0
nodes [{id=Modulator_0, children [WavePlayer_0]}, {id=WavePlayer_0, properties {soundWaveAssetPtr=&/PortaShack_Cook}}]
properties {
 attenuationSettings=/Game/Sounds/Fort_Weapons/Grenades/Grenade_Default_Cook_Att
 maxDistance=3000
 soundClassObject=&Fort_Gadgets
 volumeMultiplier=0.4875
}
