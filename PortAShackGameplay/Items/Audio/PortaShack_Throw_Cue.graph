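# Sound Cue for the throw sound of the Port-a-Shack gameplay item.
# A single wave player plays the PortaShack_Throw wave through a root Modulator left at its default pitch and volume variation.
# The cue runs at reduced volume in the Fort_Gadgets sound class and borrows the shared Fort_Traps_Att attenuation for distance falloff.
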SoundCue /PortAShackGameplay/Items/Audio/PortaShack_Throw_Cue
root=Modulator_0
nodes [
 {id=Modulator_0, children [WavePlayer_1]}
 {id=WavePlayer_1, properties {soundWaveAssetPtr=&/PortaShack_Throw}}
]
properties {
 attenuationSettings=&Fort_Traps_Att
 maxDistance=6000
 soundClassObject=&Fort_Gadgets
 volumeMultiplier=0.4875
}
