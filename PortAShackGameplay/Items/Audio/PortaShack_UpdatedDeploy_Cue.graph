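# Sound cue for the Port-a-Shack deploy sound, playing one of four build sound variations.
# An equally weighted Random node, excluded from branch culling, picks the wave and feeds a root Modulator that pins pitch at its original value.
# The cue uses the Port-a-Shack deploy attenuation asset and the gadgets sound class, with its own concurrency override capping simultaneous instances at three.
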SoundCue /PortAShackGameplay/Items/Audio/PortaShack_UpdatedDeploy_Cue
root=Modulator_4
nodes [
 {id=Modulator_4, children [Random_0], properties {pitchMax=1, pitchMin=1}}
 {
  id=Random_0
  children [WavePlayer_8, WavePlayer_13, WavePlayer_14, WavePlayer_15]
  properties {bShouldExcludeFromBranchCulling=true, weights [1,1,1,1]}
 }
 {id=WavePlayer_13, properties {soundWaveAssetPtr=&/PortaShack_Build_01}}
 {id=WavePlayer_14, properties {soundWaveAssetPtr=&/PortaShack_Build_02}}
 {id=WavePlayer_15, properties {soundWaveAssetPtr=&/PortaShack_Build_03}}
 {id=WavePlayer_8, properties {soundWaveAssetPtr=&/PortaShack_Build_04}}
]
properties {
 attenuationSettings=&/ATT_PortaShack_Deploy
 bOverrideConcurrency=true
 maxDistance=12000
 soundClassObject=&Fort_Gadgets
 volumeMultiplier=1
 attenuationOverrides {falloffDistance=6300}
 concurrencyOverrides {maxCount=3}
}
